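# Sound cue for opening a small cardboard box loot container, using foley from the searching-loop end recordings.
# An equally weighted random node picks one of three wave variations, with branch culling disabled on that node.
# The cue uses item-interact mono attenuation and the loot collect sound class.
# A concurrency override caps simultaneous instances, stopping the farthest first and then preventing new ones.
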SoundCue /Game/Sounds/Foley_Loot/Containers/CardboardBox/CardboardBox_Small_Open_Cue
root=Random_0
nodes [
 {id=Random_0, children [WavePlayer_1,WavePlayer_2,WavePlayer_3], properties {bShouldExcludeFromBranchCulling=true, weights [1,1,1]}}
 {id=WavePlayer_1, properties {soundWaveAssetPtr=&/CardboardBox_Small_Searching_Loop_End_01}}
 {id=WavePlayer_2, properties {soundWaveAssetPtr=&/CardboardBox_Small_Searching_Loop_End_02}}
 {id=WavePlayer_3, properties {soundWaveAssetPtr=&/CardboardBox_Small_Searching_Loop_End_03}}
]
properties {
 attenuationSettings=/Game/Sounds/Attenuation/Fort_Item_Interact_Mono_Att
 bOverrideConcurrency=true
 maxDistance=1500
 soundClassObject=&Fort_Loot_Collect
 concurrencyOverrides {maxCount=4, resolutionRule=StopFarthestThenPreventNew}
}
